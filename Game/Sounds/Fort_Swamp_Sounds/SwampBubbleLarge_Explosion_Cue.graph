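# Sound cue for a large swamp bubble explosion, routed to the ambient tile-mix sound class at a low volume.
# A Random node picks one of five equally weighted MudBlob waves, then a Modulator varies its pitch slightly around normal.
# The cue overrides attenuation for a limited falloff range and overrides concurrency so new instances are refused once the limit is reached.
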SoundCue /Game/Sounds/Fort_Swamp_Sounds/SwampBubbleLarge_Explosion_Cue
root=Modulator_0
nodes [
 {id=Modulator_0, children [Random_1], properties {pitchMax=1.1, pitchMin=0.9}}
 {id=Random_1, children [WavePlayer_5,WavePlayer_6,WavePlayer_7,WavePlayer_8,WavePlayer_9], properties {weights [1,1,1,1,1]}}
 {id=WavePlayer_5, properties {soundWaveAssetPtr=&/MudBlobB002}}
 {id=WavePlayer_6, properties {soundWaveAssetPtr=&/MudBlobB004}}
 {id=WavePlayer_7, properties {soundWaveAssetPtr=&/MudBlobB001}}
 {id=WavePlayer_8, properties {soundWaveAssetPtr=&/MudBlobB003}}
 {id=WavePlayer_9, properties {soundWaveAssetPtr=&/MudBlobB005}}
]
properties {
 bOverrideAttenuation=true
 bOverrideConcurrency=true
 maxDistance=2200
 soundClassObject=/Game/Sounds/FortSoundClassesAndModes/Fort_Ambient_TileMix
 volumeMultiplier=0.15
 attenuationOverrides {falloffDistance=2000, nonSpatializedRadiusStart=100, attenuationShapeExtents {x=200}}
 concurrencyOverrides {resolutionRule=PreventNew}
}
