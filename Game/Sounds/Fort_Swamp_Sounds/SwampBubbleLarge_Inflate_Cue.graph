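# Sound cue for a large swamp bubble inflating, playing a quiet mud blob one-shot in the ambient tile mix sound class.
# An equally weighted random node picks one of five MudBlob waves, and a modulator slightly varies its pitch.
# The cue overrides attenuation for distance falloff and overrides concurrency so new instances are blocked once the limit is reached.
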SoundCue /Game/Sounds/Fort_Swamp_Sounds/SwampBubbleLarge_Inflate_Cue
root=Modulator_0
nodes [
 {id=Modulator_0, children [Random_0], properties {pitchMax=1.1, pitchMin=0.9}}
 {id=Random_0, children [WavePlayer_0,WavePlayer_1,WavePlayer_2,WavePlayer_3,WavePlayer_4], properties {weights [1,1,1,1,1]}}
 {id=WavePlayer_0, properties {soundWaveAssetPtr=&/MudBlobA001}}
 {id=WavePlayer_1, properties {soundWaveAssetPtr=&/MudBlobA002}}
 {id=WavePlayer_2, properties {soundWaveAssetPtr=&/MudBlobA003}}
 {id=WavePlayer_3, properties {soundWaveAssetPtr=&/MudBlobA004}}
 {id=WavePlayer_4, properties {soundWaveAssetPtr=&/MudBlobA005}}
]
properties {
 bOverrideAttenuation=true
 bOverrideConcurrency=true
 maxDistance=2200
 soundClassObject=/Game/Sounds/FortSoundClassesAndModes/Fort_Ambient_TileMix
 volumeMultiplier=0.2
 attenuationOverrides {falloffDistance=2000, nonSpatializedRadiusStart=100, attenuationShapeExtents {x=200}}
 concurrencyOverrides {resolutionRule=PreventNew}
}
